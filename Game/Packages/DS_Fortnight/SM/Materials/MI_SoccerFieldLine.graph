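# Masked, lit base material (despite its MI_ prefix) for soccer field line markings in the DS_Fortnight package.
# A tinted diffuse texture feeds base color, a normal map adds detail, and the MaskTexture alpha drives the opacity mask.
# An opacity scalar scales that alpha into Opacity only, not into the mask that Masked blending uses.
# The combined material attributes pass through MF_CreativeMoveTool.
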Material /Game/Packages/DS_Fortnight/SM/Materials/MI_SoccerFieldLine
graph {
 properties {bUseMaterialAttributes=true, blendMode=BLEND_Masked, pixelDepthOffsetMode=PDOM_Legacy, usedShadingModels=Default Lit}
 nodes [
  Constant3Vector()
  n1 = TextureSampleParameter2D() {parameterName=MaskTexture, texture=/Game/Packages/DS_Fortnight/SM/Textures/T_soccer_field_line_01_D}
  n2 = ScalarParameter() {defaultValue=0.5, parameterName=opacity}
  n3 = SetMaterialAttributes!(
   MaterialAttributes
   Base Color=Multiply(
    A=TextureSampleParameter2D() {parameterName=Diffuse, texture=&T_soccer_field_line_01_D}.RGB
    B=Constant3Vector() {constant=<rgba 0.3404 0.375 0.3365 0>}
   )
   Opacity=Saturate(None=Multiply(A=@n1.A, B=@n2))
   Opacity Mask=@n1.A
   Normal=TextureSampleParameter2D() {parameterName=Normal, samplerType=SAMPLERTYPE_Normal, texture=/Game/Packages/DS_Fortnight/SM/Textures/T_soccer_field_line_01_N}.RGB
  ) {attributeSetTypes [
    #1
    #2
    #3
    #4
   ]}
  MaterialFunctionCall(MA In=@n3) {materialFunction=&MF_CreativeMoveTool}
 ]
}
