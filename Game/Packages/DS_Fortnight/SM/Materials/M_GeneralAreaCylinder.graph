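# Translucent lit material, named for a general area cylinder, with a light blue color and a time-driven pulse.
# It uses volumetric per-vertex non-directional translucency lighting.
# A sine of time is scaled, offset and halved to oscillate gently around one half.
# Constant color, scalar and vertex color nodes sit alongside that chain, with no links to material outputs shown.
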Material /Game/Packages/DS_Fortnight/SM/Materials/M_GeneralAreaCylinder
graph {
 properties {
  blendMode=BLEND_Translucent
  pixelDepthOffsetMode=PDOM_Legacy
  translucencyLightingMode=TLM_VolumetricPerVertexNonDirectional
  usedShadingModels=Default Lit
 }
 nodes [
  Constant3Vector() {constant=<rgba 0.188 0.6939 1 1>}
  Constant() {r=0.75}
  n1 = Sine(None=Multiply(A=Time(), B=Constant() {r=1}))
  n2 = Constant() {r=0.35}
  Divide(A=Add(A=Multiply(A=@n1, B=@n2), B=Constant() {r=1}), B=Constant() {r=2})
  VertexColor()
 ]
}
