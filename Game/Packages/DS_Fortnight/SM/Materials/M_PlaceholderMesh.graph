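# M_PlaceholderMesh is a simple Default Lit material, named for use on placeholder meshes in the DS_Fortnight package.
# It declares three parameters for material instances: a mid-grey Base Color, a mostly metallic Metallic value and a medium Roughness, with no textures or math nodes.
# It is flagged for use with Nanite and assigns the GateDevice physical material.
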Material /Game/Packages/DS_Fortnight/SM/Materials/M_PlaceholderMesh
graph {
 properties {
  bUsedWithNanite=true
  physMaterial=/Game/FortressPhysicalMaterials/GateDevice
  pixelDepthOffsetMode=PDOM_Legacy
  usedShadingModels=Default Lit
 }
 nodes [
  ScalarParameter() {defaultValue=0.75, parameterName=Metallic}
  ScalarParameter() {defaultValue=0.5, parameterName=Roughness}
  VectorParameter() {parameterName=Base Color, defaultValue=<rgba 0.5 0.5 0.5 1>}
 ]
}
