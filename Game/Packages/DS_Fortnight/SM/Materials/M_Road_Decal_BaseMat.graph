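# Deferred decal base material for road markings, using alpha-composite blending and a parking spot mask texture by default.
# The decal texture's red channel lerps between low and high opacity parameters, then saturates into a mask.
# That mask premultiplies the color, specular and roughness parameters.
# Two shading path switches split the color: one carries it on the default path, the other carries it multiplied by the mask again on mobile.
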Material /Game/Packages/DS_Fortnight/SM/Materials/M_Road_Decal_BaseMat
graph {
 properties {blendMode=BLEND_AlphaComposite, materialDomain=MD_DeferredDecal, pixelDepthOffsetMode=PDOM_Legacy, usedShadingModels=Default Lit}
 nodes [
  Multiply()
  n1 = Saturate(None=LinearInterpolate(
    A=ScalarParameter() {defaultValue=-0.1, parameterName=Opacity_Low}
    B=ScalarParameter() {defaultValue=0.84, parameterName=Opacity_High}
    Alpha=TextureSampleParameter2D() {parameterName=Decal, samplerType=SAMPLERTYPE_LinearColor, texture=/Game/Packages/DS_Fortnight/T_FORT_ParkingSpotMasks}.R
   ))
  Multiply(A=ScalarParameter() {defaultValue=0.5, parameterName=Specular}, B=@n1)
  Multiply(A=@n1, B=ScalarParameter() {defaultValue=0.3, parameterName=Roughness})
  n2 = Multiply(A=VectorParameter() {parameterName=Color, defaultValue=<rgba 0.6562 0.6562 0.6562 1>}.RGB, B=@n1)
  ShadingPathSwitch(Default=@n2, Mobile=Constant3Vector()) {desc=Migrated from FeatureLevelSwitch}
  ShadingPathSwitch(Default=Constant(), Mobile=Multiply(A=@n2, B=@n1)) {desc=Migrated from FeatureLevelSwitch}
 ]
}
