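# M_VertexColor is a Default Lit material built around mesh vertex color.
# A single Vertex Color node feeds the A, B and Alpha inputs of a Lerp.
# The Lerp result drives the DiffuseColor input of the ConvertFromDiffSpec material function, with no other function inputs set.
# The graph shows no links from that function to the material outputs.
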Material /Game/Packages/DS_Fortnight/SM/Materials/M_VertexColor
graph {
 properties {pixelDepthOffsetMode=PDOM_Legacy, usedShadingModels=Default Lit}
 nodes [
  n1 = VertexColor()
  MaterialFunctionCall(DiffuseColor=LinearInterpolate(A=@n1, B=@n1, Alpha=@n1)) {materialFunction=&ConvertFromDiffSpec}
 ]
}
