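# Default Lit material that defines a green surface color varying with the mesh's vertex colors.
# A lerp blends from pale green to saturated yellowish green, using vertex color as the alpha.
# The result feeds only the DiffuseColor input of the ConvertFromDiffSpec function, which converts diffuse/specular values to metallic-workflow values.
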Material /Game/Packages/DS_Fortnight/SM/Materials/M_VertexColor_Green
graph {
 properties {pixelDepthOffsetMode=PDOM_Legacy, usedShadingModels=Default Lit}
 nodes [MaterialFunctionCall(DiffuseColor=LinearInterpolate(A=Constant3Vector() {constant=<rgba 0.6867 1 0.6 1>}, B=Constant3Vector() {constant=<rgba 0.2167 1 0 1>}, Alpha=VertexColor())) {materialFunction=&ConvertFromDiffSpec}]
}
