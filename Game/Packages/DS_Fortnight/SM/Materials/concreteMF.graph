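# Material function that builds a concrete/road layer for blending, outputting material attributes and a height-based lerp alpha driven by vertex red.
# Static switches select world-aligned or UV-based sampling of diffuse, spec mask and normal textures.
# The layer's opacity feeds two HeightLerps, whose opposing alphas multiply into a mask that tints the transition black.
# Specular and roughness are then remapped, with optional color-derived specular and flat roughness; an EditSpecTrue branch is unconnected.
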MaterialFunction /Game/Packages/DS_Fortnight/SM/Materials/concreteMF
graph {nodes [
  n1 = Constant() {desc=FalloffContrast, r=1}
  n2 = TextureObjectParameter() {
   group=Concrete
   parameterName=Concrete
   samplerSource=SSM_Wrap_WorldGroupSettings
   texture=/Game/Packages/DS_Fortnight/SM/Textures/T_Street_Road_D
  }
  n3 = TextureObjectParameter() {
   group=Concrete
   parameterName=Concrete_SpecMask
   samplerSource=SSM_Wrap_WorldGroupSettings
   samplerType=SAMPLERTYPE_Masks
   texture=/Game/Packages/DS_Fortnight/SM/Textures/T_Street_Road_M
  }
  n4 = TextureObjectParameter() {
   group=Concrete
   parameterName=ConcreteTextureNormal
   samplerSource=SSM_Wrap_WorldGroupSettings
   samplerType=SAMPLERTYPE_Normal
   texture=/Game/Packages/DS_Fortnight/SM/Textures/T_Street_Road_N
  }
  n5 = MaterialFunctionCall!(
   Diffuse Texture Object=FunctionInput(Preview=@n2) {bUsePreviewValueAsDefault=true, inputName=Concrete diffuse, inputType=FunctionInput_Texture2D}
   Spec Texture Object=FunctionInput(Preview=@n3) {bUsePreviewValueAsDefault=true, inputName=Concrete Spec, inputType=FunctionInput_Texture2D, sortPriority=1}
   TextureSize=ScalarParameter() {defaultValue=768, group=Concrete, parameterName=ConcTexSize}
   --------------------------
   Use seperate settings for Z projection
   Use seperate texture for Z projection
   Texture Size Z projection
   Z Plane Diffuse Texture Object
   Z Plane Spec Texture Object
   "--------------------------      "
   ProjectionTransitionContrast
   Use Texture Alpha Heightmap to Refine the Transistion
   HeightLerpTransitionContrast
   PickTheDiffuseHeightChannel
   TheDiffuseHeightChannel
   "--------------------------       "
   NormalTextureObject=FunctionInput(Preview=@n4) {bUsePreviewValueAsDefault=true, inputName=Concrete Normal, inputType=FunctionInput_Texture2D, sortPriority=2}
   World Space Normals
   "--------------------------  "
   Export Float 4 Diffuse=StaticBool() {value=true}
   Export Float 4 Spec
   "--------------------------     "
   WorldPosition
   World Space Normal
  ) {materialFunction=/Engine/Functions/Engine_MaterialFunctions02/Texturing/WorldAlignedTextures_Complex}
  n6 = TextureCoordinate()
  n7 = VectorParameter() {group=Concrete, parameterName=ConcreteTextureCoordinateControls_RG_Scale_BA_Offset, defaultValue=<rgba 1 1 0 0>}
  n8 = ComponentMask(None=@n7.RGB) {g=true, r=true}
  n9 = StaticSwitchParameter(True=Add(A=Multiply(A=@n6, B=@n8), B=AppendVector(A=@n7.B, B=@n7.B)), False=@n6) {group=Concrete, parameterName=Modify Concrete UVS}
  n10 = TextureSampleParameter2D(UVs=@n9) {
   group=Concrete
   parameterName=Concrete
   samplerSource=SSM_Wrap_WorldGroupSettings
   texture=&T_Street_Road_D
  }
  n11 = TextureSampleParameter2D(UVs=@n9) {
   group=Concrete
   parameterName=Concrete_SpecMask
   samplerSource=SSM_Wrap_WorldGroupSettings
   samplerType=SAMPLERTYPE_Masks
   texture=&T_Street_Road_M
  }
  n12 = TextureSampleParameter2D(UVs=@n9) {
   group=Concrete
   parameterName=ConcreteTextureNormal
   samplerSource=SSM_Wrap_WorldGroupSettings
   samplerType=SAMPLERTYPE_Normal
   texture=&T_Street_Road_N
  }
  n13 = Transform(None=@n12.RGB)
  n14 = VertexNormalWS()
  n15 = StaticBoolParameter() {defaultValue=true, group=Concrete, parameterName=Use Concrete Normals}
  n16 = StaticSwitchParameter(
   True=MaterialFunctionCall!(In=StaticSwitchParameter(True=@n5.XYZ Mat Attributes - See Tooltip, False=@n5.XYZ Mat Attributes - See Tooltip) {group=Concrete, parameterName=UseWorldPositionTriPlanar}) {materialFunction=/Game/Packages/Fortress_SharedMaterials/Base_Material_Functions/SwizzleWorldAlignedSpecOutputForFortnite}.Result
   False=MakeMaterialAttributes!(
    BaseColor=@n10.RGB
    Metallic=@n11.G
    Specular=@n11.G
    Roughness=@n11.G
    Anisotropy
    EmissiveColor
    Opacity=@n10.RGB
    OpacityMask
    Normal=StaticSwitch(True=@n13, False=@n14, Value=@n15)
    Tangent
    WorldPositionOffset
    SubsurfaceColor
    ClearCoat
    ClearCoatRoughness
    AmbientOcclusion
    Refraction
    CustomizedUVs_0
    CustomizedUVs_1
    CustomizedUVs_2
    CustomizedUVs_3
    CustomizedUVs_4
    CustomizedUVs_5
    CustomizedUVs_6
    CustomizedUVs_7
    PixelDepthOffset
    ShadingModel
    Displacement
   )
  ) {defaultValue=true, group=Concrete, parameterName=UseWorldPositionCoordinatesForConcrete}
  n17 = MaterialFunctionCall!(Mat attributes=@n16) {materialFunction=/Engine/Functions/Engine_MaterialFunctions02/MatLayerBlend_BreakOpacity}
  n18 = VertexColor()
  n19 = Clamp(None=LinearInterpolate(
    A=Constant() {desc=FalloffPushLower, r=-0.5}
    B=Constant() {desc=FalloffPushUpper, r=3}
    Alpha=MaterialFunctionCall!(Lerp A, Lerp B, Height Texture=@n17.Result, Falloff Contrast=@n1, Transition Phase=@n18.R) {materialFunction=/Game/Packages/Fortnight_MaterialFunctions/Temp/HeightLerp}.Lerp Alpha
   ))
  n20 = LinearInterpolate(
   A=Constant() {desc=FalloffPushUpper, r=1}
   B=Constant() {desc=FalloffPushLower, r=-0.5}
   Alpha=MaterialFunctionCall(Height Texture=@n17.Result, Transition Phase=@n18.R) {materialFunction=&HeightLerp}.Lerp Alpha No Contrast
  )
  n21 = GetMaterialAttributes(None=@n16) {attributeGetTypes [#1]}
  n22 = SetMaterialAttributes!(
   MaterialAttributes=@n21.MaterialAttributes
   Specular=LinearInterpolate(
    A=ScalarParameter() {defaultValue=-0.35, group=Concrete, parameterName=Road_Spec}
    B=ScalarParameter() {defaultValue=4, group=Concrete, parameterName=Road_Spec_1}
    Alpha=ComponentMask(None=@n21.BaseColor) {r=true}
   )
  ) {attributeSetTypes [#2]}
  n23 = MaterialFunctionCall!(
   Material=MaterialFunctionCall!(
    Mat attributes=StaticSwitchParameter(True=@n22, False=@n16) {group=Concrete, parameterName=UseDerivedConcreteSpecular}
    Use Spec=StaticBoolParameter() {defaultValue=true, group=Concrete, parameterName=Use Concrete Spec}
    Use Normals=StaticBoolParameter() {defaultValue=true, group=Concrete, parameterName=Use Concrete Normals}
   ) {materialFunction=/Game/Packages/Fortnite_UniqueMaterials_1/MaterialFunctions/Fortnite_Terrain_invertOpacityNormalBool}.Result
   Tint=Constant()
   TintMask=Multiply(A=@n19, B=Clamp(None=@n20))
  ) {materialFunction=/Engine/Functions/MaterialLayerFunctions/MatLayerBlend_Tint}
  n24 = GetMaterialAttributes(None=@n23.Blended Material) {attributeGetTypes [#3, #2]}
  n25 = Constant() {r=-0.2}
  n26 = Constant() {r=1}
  n27 = LinearInterpolate(A=Constant() {r=0.2}, B=Constant() {r=0.5}, Alpha=@n24.Specular)
  n28 = SetMaterialAttributes!(
   MaterialAttributes=@n24.MaterialAttributes
   Roughness=Saturate(None=LinearInterpolate(A=@n25, B=@n26, Alpha=@n24.Roughness))
   Specular=StaticSwitchParameter(True=Saturate(None=@n27), False=@n24.Specular) {defaultValue=true, desc=disable for Figment S02, group=Concrete, parameterName=ApplySpecularRemap}
  ) {attributeSetTypes [#3, #2]}
  n29 = ScalarParameter() {defaultValue=0.85, parameterName=FlatRoughness}
  FunctionOutput(None=StaticSwitchParameter(
    True=MaterialFunctionCall!(Material=@n28, NewRoughness=@n29, Mask) {materialFunction=/Engine/Functions/MaterialLayerFunctions/MatLayerBlend_RoughnessOverride}.Blended Material
    False=@n28
   ) {group=Concrete, parameterName=UseFlatRoughness}) {outputName=Material Attribute}
  FunctionOutput(None=@n19) {outputName="Lerp Alpha (controlled by the red vertex color)", sortPriority=1}
  LinearInterpolate(
   A=ScalarParameter() {defaultValue=-0.35, group=Concrete, parameterName=Gravel_Spec}
   B=ScalarParameter() {defaultValue=4, group=Concrete, parameterName=Gravel_Spec_1}
  )
  n30 = GetMaterialAttributes(None=@n23.Blended Material) {attributeGetTypes [#3, #2]}
  n31 = ScalarParameter() {group=Concrete, parameterName=SpecularRoughnessMin, sortPriority=9}
  n32 = ScalarParameter() {defaultValue=1, group=Concrete, parameterName=SpecularRoughnessMax, sortPriority=8}
  n33 = ComponentMask(None=@n30.Roughness) {r=true}
  n34 = ScalarParameter() {defaultValue=1, group=Concrete, parameterName=SpecularMult, sortPriority=10}
  StaticSwitchParameter(
   True=SetMaterialAttributes(
    MaterialAttributes=@n30.MaterialAttributes
    Roughness=LinearInterpolate(A=@n31, B=@n32, Alpha=@n33)
    Specular=Multiply(A=@n30.Specular, B=@n34)
   ) {attributeSetTypes [#3, #2]}
   False=@n23.Blended Material
  ) {group=Concrete, parameterName=EditSpecTrue}
 ]}
